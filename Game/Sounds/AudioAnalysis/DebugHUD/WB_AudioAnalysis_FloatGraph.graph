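# Debug HUD widget that live-graphs one float from a Material Parameter Collection for audio analysis.
# Initialize MPC sets the collection, parameter and VectorIndex, where -1 reads a scalar and other values pick a vector channel.
# Tick sizes the graph once from widget geometry, then updates text, progress bar and a rolling point buffer.
# OnPaint draws that history as a line over a box.
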WidgetBlueprint /Game/Sounds/AudioAnalysis/DebugHUD/WB_AudioAnalysis_FloatGraph
graphs [
 EventGraph {nodes [
   n1 = Initialize MPC!() -> (OutputDelegate, then, Parameter:Name, Collection:Material Parameter Collection Object Reference, VectorIndex:int=-1) {is=title K2Node_CustomEvent}
   > Set Parameter(Parameter:Name=@n1.Parameter) -> (Output_Get:Name=None)
   > Set VectorIndex(VectorIndex:int=@n1.VectorIndex) -> (Output_Get:int=0)
   > Set Collection(Collection:Material Parameter Collection Object Reference=@n1.Collection) -> (Output_Get:Material Parameter Collection Object Reference)
   n2 = Event PreConstruct()
   Set calculated(execute=[@n2.then, @n3.then], calculated:bool=false) -> (Output_Get:bool)
   K2Node_Composite_0 = GetLabelName!() -> (ReturnValue:Text) {is=title K2Node_Composite}
   > "SetText (Text)"(self=Get Label_() -> (Label_:Text Object Reference), InText=@K2Node_Composite_0)
   n3 = Event Construct()
   n4 = Event Tick()
   > n5 = Branch(Condition=Get calculated() -> (calculated:bool))
   n6 = Branch(execute=[@n5.then, @n11.then], Condition="Equal (Integer)"(A=Get VectorIndex() -> (VectorIndex:int=0), B=-1) {is=title K2Node_CallFunction})
   > n7 = GetScalarParameterValue!(
    execute
    Collection:Material Parameter Collection Object Reference=Get Collection() -> (Collection:Material Parameter Collection Object Reference)
    ParameterName:Name=Get Parameter() -> (Parameter:Name=None)
   ) -> (then, ReturnValue:float) {is=title K2Node_CallMaterialParameterCollectionFunction}
   > "SetText (Text)"(
    self=Get Value() -> (Value:Text Object Reference)
    InText="To Text (Float)"(Value=@n7, MinimumIntegralDigits=1, MaximumIntegralDigits=324, MinimumFractionalDigits=0, MaximumFractionalDigits=2)
   )
   > SetPercent!(execute, self:Progress Bar Object Reference=Get progressBar() -> (progressBar:Progress Bar Object Reference), InPercent:float=@n7) -> (then)
   > Set Value!(execute, self, value:double=@n7) -> (then) {is=title K2Node_CallFunction}
   n8 = GetVectorParameterValue!(
    execute=@n6.else
    Collection:Material Parameter Collection Object Reference=Get Collection()
    ParameterName:Name=Get Parameter() -> (Parameter=None)
   ) -> (then, ReturnValue:Linear Color Structure) {is=title K2Node_CallMaterialParameterCollectionFunction}
   n9 = BreakColor(InColor=@n8)
   n10 = Select!(Option 0:double=@n9, Option 1:double=@n9.G, Option 2:double=@n9.B, Option 3:double=@n9.A, Index:int=Get VectorIndex() -> (VectorIndex=0)) -> (ReturnValue:double)
   > "SetText (Text)"(
    self=Get Value()
    InText="To Text (Float)"(Value=@n10, MinimumIntegralDigits=1, MaximumIntegralDigits=324, MinimumFractionalDigits=0, MaximumFractionalDigits=2)
   )
   > SetPercent(self=Get progressBar(), InPercent=@n10)
   > Set Value(value=@n10)
   calcWidgetSize!(execute=@n5.else, self, geom:Geometry Structure=@n4.MyGeometry) -> (then)
   > n11 = Set calculated(calculated=true)
  ]}
 GetLabelName {
  path=EventGraph.K2Node_Composite_0.GetLabelName
  nodes [
   Inputs!()
   n1 = "To String (Name)"(InName=Get Parameter() -> (Parameter=None))
   n2 = Get VectorIndex() -> (VectorIndex=0)
   n3 = SelectString(A="To String (Integer)"(InInt=@n2), bPickA="Not Equal (Integer)"!(A:int=@n2, B:int=-1) -> (ReturnValue:bool) {is=title K2Node_CallFunction})
   Outputs!(ReturnValue:Text="To Text (String)"(InString=Append(A=@n1, B=@n3)))
  ]
 }
 OnPaint {nodes [
   n1 = OnPaint!() -> (then, Context:"Paint Context Structure (by ref)") {is=title K2Node_FunctionEntry}
   n2 = Get!() -> (size_X:double, size_Y:double)
   n3 = DrawBox!(
    execute
    Context:"Paint Context Structure (by ref)"=@n1
    Position:Vector 2D Structure=Get position() -> (position:Vector 2D Structure)
    Size_X:double=@n2
    Size_Y:double=@n2.size_Y
    Brush:Slate Brush Asset Object Reference=/Game/UI/CodeReferenced/Tooltip-Micro_Brush
    Tint:Linear Color Structure=Get color() -> (color:Linear Color Structure)
   ) -> (then)
   DrawLines!(
    execute=[@n3.then, @n1.then]
    Context:"Paint Context Structure (by ref)"=@n1
    Points:Array of Vector 2D Structures=Get Points() -> (Points:Array of Vector 2D Structures)
    Tint:Linear Color Structure=(R=1,G=1,B=1,A=1)
    bAntiAlias:bool=true
    Thickness:float=1
   ) -> (then)
   > Return Node()
  ]}
 Set Value {nodes [
   n1 = Set Value!() -> (then, value:double=~) {is=title K2Node_FunctionEntry}
   n2 = Get Points()
   n3 = Length(TargetArray:Array of Vector 2D Structures=@n2)
   > n4 = Branch(Condition="integer > integer"(A=@n3, B=Get numPoints() -> (numPoints:int=0)) {is=title K2Node_CallFunction})
   > Remove Index!(execute, TargetArray:Array of Vector 2D Structures=@n2, IndexToRemove:int=0) -> (then)
   n5 = float * float(A=float - float(B=@n1) {is=title K2Node_CallFunction}, B=Get().size_Y)
   > Add!(execute, TargetArray:Array of Vector 2D Structures=@n2, NewItem_X:double=0, NewItem_Y:double=@n5) -> (then, ReturnValue) {is=title K2Node_CallArrayFunction}
   n6 = Get Points()
   > n7 = For Each Loop!(Exec, Array:Array of Vector 2D Structures=@n6) -> (LoopBody, Array Element_X:double, Array Element_Y:double, Array Index, Completed)
   n8 = "To Float (Integer)"(InInt=Get numPoints() -> (numPoints=0))
   n9 = Get()
   > Set Array Elem!(
    execute
    TargetArray:Array of Vector 2D Structures=@n6
    Index:int=@n7.Array Index
    Item_X:double=int * float!(A:int=@n7.Array Index, B:double=float / float(A=@n9, B=@n8) {is=title K2Node_CallFunction}) -> (ReturnValue:double)
    Item_Y:double=@n7.Array Element_Y
    bSizeToFit:bool=false
   ) -> (then)
   n10 = "To Float (Integer)"(InInt=Get numPoints() -> (numPoints=0))
   Add(execute=@n4.else, TargetArray=@n2, NewItem_X=float * float(A="To Float (Integer)"(InInt=@n3), B=float / float(A=@n9, B=@n10)), NewItem_Y=@n5)
  ]}
 calcWidgetSize {nodes [
   n1 = calcWidgetSize!() -> (then, geom:Geometry Structure) {is=title K2Node_FunctionEntry}
   > Set position(position:Vector 2D Structure=(X=0,Y=0)) -> (Output_Get:Vector 2D Structure)
   n2 = BreakVector2D(InVec=GetLocalSize!(Geometry:"Geometry Structure (by ref)"=@n1) -> (ReturnValue:Vector 2D Structure))
   > Set!(execute, size_X:double=float - float(A=@n2, B=40), size_Y:double=float - float(A=@n2.Y, B=30)) -> (then, Output_Get:Vector 2D Structure)
  ]}
]
variables [
 calculated {type=boolean}
 Collection {type=MaterialParameterCollection, default=null}
 color {type=LinearColor, default=<rgba 0 0 0 0.5>}
 numPoints {type=integer, default=50}
 Parameter {type=string, default=Emote_AudioLow}
 Points {type="array<Vector2D>", default []}
 position {type=Vector2D, default=<4 4>}
 size {type=Vector2D, default=<100 100>}
 VectorIndex {type=integer, default=-1}
]
widgets {
 info {inheritedWidgetCount=0, parentClass=&UserWidget, rootWidgetClass=&CanvasPanel, widgetCount=7}
 widgets [
  CanvasPanel_0 {widgetClassPath=&CanvasPanel, bIsVariable=true, widget=CanvasPanel_0}
  CanvasPanel_59 {
   widgetClassPath=&CanvasPanel
   parent=CanvasPanel_0
   widget=CanvasPanel_59
   slot {class=CanvasPanelSlot, properties {layoutData {alignment=<0 0>, anchors {maximum=<0 0>, minimum=<0 0>}, offsets=<ltrb 0 0 250 110>}}}
  }
  CanvasPanel_30 {
   widgetClassPath=&CanvasPanel
   parent=CanvasPanel_59
   widget=CanvasPanel_30
   slot {class=CanvasPanelSlot, properties {layoutData {alignment=<0 0>, anchors {maximum=<0 0>, minimum=<0 0>}, offsets=<ltrb 0 0 250 100>}}}
  }
  Label_ {
   widgetClassPath=&TextBlock
   bIsVariable=true
   parent=CanvasPanel_30
   widget=Label_
   properties {
    text=Label
    font {
     bForceMonospaced=false
     bMaterialIsStencil=false
     fontMaterial=null
     fontObject=&Roboto
     letterSpacing=0
     monospacedWidth=1
     size=10
     skewAmount=0
     typefaceFontName=Regular
     outlineSettings {
      bApplyOutlineToDropShadows=false
      bMiteredCorners=false
      bSeparateFillAlpha=false
      outlineMaterial=null
      outlineSize=0
      outlineColor=<rgba 0 0 0 1>
     }
    }
   }
   slot {
    class=CanvasPanelSlot
    properties {bAutoSize=true, zOrder=1, layoutData {alignment=<0 1>, anchors {maximum=<0 1>, minimum=<0 1>}, offsets=<ltrb 2 0 100 70>}}
   }
  }
  Value {
   widgetClassPath=&TextBlock
   bIsVariable=true
   parent=CanvasPanel_30
   widget=Value
   properties {
    justification=Center
    text=0.00
    font {
     bForceMonospaced=false
     bMaterialIsStencil=false
     fontMaterial=null
     fontObject=&Roboto
     letterSpacing=0
     monospacedWidth=1
     size=10
     skewAmount=0
     typefaceFontName=Regular
     outlineSettings {
      bApplyOutlineToDropShadows=false
      bMiteredCorners=false
      bSeparateFillAlpha=false
      outlineMaterial=null
      outlineSize=0
      outlineColor=<rgba 0 0 0 1>
     }
    }
   }
   slot {
    class=CanvasPanelSlot
    properties {bAutoSize=true, layoutData {alignment=<1 1>, anchors {maximum=<1 1>, minimum=<1 1>}, offsets=<ltrb -2 0 100 20>}}
   }
  }
  progressBar {
   widgetClassPath=/Script/UMG.ProgressBar
   bIsVariable=true
   parent=CanvasPanel_30
   widget=progressBar
   properties {
    barFillType=BottomToTop
    fillColorAndOpacity=<rgba 0.2396 0.25 0.3 1>
    widgetStyle {
     enableFillAnimation=false
     backgroundImage {
      bIsDynamicallyLoaded=false
      drawAs=Box
      imageType=FullColor
      mirroring=NoMirror
      resourceName=../../../Engine/Content/Slate/Common/ProgressBar_Background.png
      resourceObject=null
      tiling=NoTile
      imageSize=<0 0>
      margin=<ltrb 0.4167 0.4167 0.4167 0.4167>
      outlineSettings {
       bUseBrushTransparency=false
       roundingType=HalfHeightRadius
       width=0
       color {colorUseRule=UseColor_Specified, specifiedColor=<rgba 0 0 0 0>}
       cornerRadii=<0 0 0 0>
      }
      tintColor {colorUseRule=UseColor_Specified, specifiedColor=<rgba 0 0 0 1>}
      uVRegion {bIsValid=false, max=<0 0>, min=<0 0>}
     }
     fillImage {
      bIsDynamicallyLoaded=false
      drawAs=Box
      imageType=FullColor
      mirroring=NoMirror
      resourceName=../../../Engine/Content/Slate/Common/ProgressBar_Fill.png
      resourceObject=null
      tiling=NoTile
      imageSize=<0 0>
      margin=<ltrb 0.4167 0.4167 0.4167 0.4167>
      outlineSettings {
       bUseBrushTransparency=false
       roundingType=HalfHeightRadius
       width=0
       color {colorUseRule=UseColor_Specified, specifiedColor=<rgba 0 0 0 0>}
       cornerRadii=<0 0 0 0>
      }
      tintColor {colorUseRule=UseColor_Specified, specifiedColor=<rgba 1 1 1 1>}
      uVRegion {bIsValid=false, max=<0 0>, min=<0 0>}
     }
     marqueeImage {
      bIsDynamicallyLoaded=false
      drawAs=Image
      imageType=FullColor
      mirroring=NoMirror
      resourceName=../../../Engine/Content/Slate/Common/ProgressBar_Marquee.png
      resourceObject=null
      tiling=Horizontal
      imageSize=<20 12>
      margin=<ltrb 0 0 0 0>
      outlineSettings {
       bUseBrushTransparency=false
       roundingType=HalfHeightRadius
       width=0
       color {colorUseRule=UseColor_Specified, specifiedColor=<rgba 0 0 0 0>}
       cornerRadii=<0 0 0 0>
      }
      tintColor {colorUseRule=UseColor_Specified, specifiedColor=<rgba 1 1 1 1>}
      uVRegion {bIsValid=false, max=<0 0>, min=<0 0>}
     }
    }
   }
   slot {class=CanvasPanelSlot, properties {layoutData {alignment=<0 0>, anchors {maximum=<1 0>, minimum=<1 0>}, offsets=<ltrb -30 0 30 100>}}}
  }
  Border_22 {
   widgetClassPath=&Border
   parent=CanvasPanel_30
   widget=Border_22
   properties {
    background {
     bIsDynamicallyLoaded=false
     drawAs=Image
     imageType=NoImage
     mirroring=NoMirror
     resourceName=None
     resourceObject=null
     tiling=NoTile
     imageSize=<32 32>
     margin=<ltrb 0 0 0 0>
     outlineSettings {
      bUseBrushTransparency=false
      roundingType=HalfHeightRadius
      width=0
      color {colorUseRule=UseColor_Specified, specifiedColor=<rgba 0 0 0 0>}
      cornerRadii=<0 0 0 0>
     }
     tintColor {colorUseRule=UseColor_Specified, specifiedColor=<rgba 0 0 0 0.5>}
     uVRegion {bIsValid=false, max=<0 0>, min=<0 0>}
    }
    contentColorAndOpacity=<rgba 0 0 0 1>
   }
   slot {
    class=CanvasPanelSlot
    properties {zOrder=-1, layoutData {alignment=<0 0>, anchors {maximum=<0 0>, minimum=<0 0>}, offsets=<ltrb 0 0 250 100>}}
   }
  }
 ]
}
