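# Debug HUD widget, based on FortAudioAnalysisDebugWidget, that stacks live float graphs for audio analysis parameters in a top-left vertical box.
# AddScalarParameter creates one WB_AudioAnalysis_FloatGraph bound to the parameter's Material Parameter Collection.
# AddVectorParameter always creates four, one per component index, ignoring VectorCount.
# ClearParameters removes the tracked graphs from their parent but never empties the Widgets array.
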WidgetBlueprint /Game/Sounds/AudioAnalysis/DebugHUD/WB_AudioAnalysis_UI
graphs [EventGraph {nodes [
   n1 = Event AddScalarParameter!() -> (OutputDelegate, then, Param:"Audio Analysis Parameter Scalar Structure (by ref)") {is=title K2Node_Event}
   > n2 = Create Widget!(
    execute:Exec
    Class:User Widget Class Reference=&/WB_AudioAnalysis_FloatGraph.WB_AudioAnalysis_FloatGraph_C
    OwningPlayer:Player Controller Object Reference
   ) -> (then:Exec, ReturnValue:WB Audio Analysis Float Graph Object Reference)
   n3 = Break Audio Analysis Parameter Scalar!(AudioAnalysisParameterScalar:"Audio Analysis Parameter Scalar Structure (by ref)"=@n1.Param) -> (
    Name:Name
    MaterialPC:Material Parameter Collection Object Reference
    NiagaraPC:Niagara Parameter Collection Object Reference
    History:Audio Analysis Param History Structure
    bDebug:bool=~
   ) {is=title K2Node_BreakStruct}
   > Initialize MPC!(
    execute
    self:WB Audio Analysis Float Graph Object Reference=@n2
    Parameter:Name=@n3
    Collection:Material Parameter Collection Object Reference=@n3.MaterialPC
    VectorIndex:int=-1
   ) -> (then)
   > AddChildToVerticalBox(self=Get GraphBox() -> (GraphBox:Vertical Box Object Reference), Content=@n2)
   > Add(
    TargetArray:Array of WB Audio Analysis Float Graph Object References=Get Widgets() -> (Widgets:Array of WB Audio Analysis Float Graph Object References)
    NewItem:"WB Audio Analysis Float Graph Object Reference (by ref)"=@n2
   ) {is=title K2Node_CallArrayFunction}
   n4 = Event AddVectorParameter!() -> (OutputDelegate, then, Param:"Audio Analysis Parameter Vector Structure (by ref)") {is=title K2Node_Event}
   > n5 = For Loop(LastIndex=3)
   > n6 = Create Widget(Class=&WB_AudioAnalysis_FloatGraph_C)
   n7 = Break Audio Analysis Parameter Vector!(AudioAnalysisParameterVector:"Audio Analysis Parameter Vector Structure (by ref)"=@n4.Param) -> (
    VectorCount:int
    GeneratedParamNames:Array of Strings
    Name:Name
    MaterialPC:Material Parameter Collection Object Reference
    NiagaraPC:Niagara Parameter Collection Object Reference
    History:Audio Analysis Param History Structure
    bDebug:bool=~
   ) {is=title K2Node_BreakStruct}
   > Initialize MPC(self=@n6, Parameter=@n7.Name, Collection=@n7.MaterialPC, VectorIndex=@n5)
   > AddChildToVerticalBox(self=Get GraphBox(), Content=@n6)
   > Add(TargetArray=Get Widgets(), NewItem=@n6)
   Event ClearParameters() {is=title K2Node_Event}
   > n8 = For Each Loop(Array:Array of WB Audio Analysis Float Graph Object References=Get Widgets()) -> (Array Element:WB Audio Analysis Float Graph Object Reference)
   > RemoveFromParent(self=@n8)
  ]}]
variables [Widgets {type="array<WB_AudioAnalysis_FloatGraph_C>", default []}]
widgets {
 info {
  inheritedWidgetCount=0
  parentClass=/Script/FortniteGame.FortAudioAnalysisDebugWidget
  rootWidgetClass=&CanvasPanel
  widgetCount=2
 }
 widgets [
  CanvasPanel_61 {widgetClassPath=&CanvasPanel, widget=CanvasPanel_61}
  GraphBox {
   widgetClassPath=&VerticalBox
   bIsVariable=true
   parent=CanvasPanel_61
   widget=GraphBox
   slot {class=CanvasPanelSlot, properties {layoutData {alignment=<0 0>, anchors {maximum=<0 0>, minimum=<0 0>}, offsets=<ltrb 20 20 250 1040>}}}
  }
 ]
}
